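# Physics asset for the Cyber Karate backpack cosmetic, defining collision bodies and joints for two dangling dynamic chains hanging off the pack.
# One large capsule covers the pack bone, and small identical capsules form a seven-link chain A and a five-link chain B.
# Each link joins its predecessor, starting at the pack, with limit values for moderate swing and tight twist.
# An angularly locked constraint joins both chain tips, closing them into a loop.
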PhysicsAsset /BRCosmetics/Accessories/FORT_Backpacks/M_MED_CyberKarate_Backpack/Meshes/M_MED_Cyber_Karate_Pack_Physics
bodies {
 dyn_chain_a_01 [{shapeType=Capsule, length=1, radius=0.8, center=<1 0 0>, rotation=<rot 90 0 0>}]
 dyn_chain_a_02 [{shapeType=Capsule, length=1, radius=0.8, center=<1 0 0>, rotation=<rot 90 0 0>}]
 dyn_chain_a_03 [{shapeType=Capsule, length=1, radius=0.8, center=<1 0 0>, rotation=<rot 90 0 0>}]
 dyn_chain_a_04 [{shapeType=Capsule, length=1, radius=0.8, center=<1 0 0>, rotation=<rot 90 0 0>}]
 dyn_chain_a_05 [{shapeType=Capsule, length=1, radius=0.8, center=<1 0 0>, rotation=<rot 90 0 0>}]
 dyn_chain_a_06 [{shapeType=Capsule, length=1, radius=0.8, center=<1 0 0>, rotation=<rot 90 0 0>}]
 dyn_chain_a_07 [{shapeType=Capsule, length=1, radius=0.8, center=<1 0 0>, rotation=<rot 90 0 0>}]
 dyn_chain_b_01 [{shapeType=Capsule, length=1, radius=0.8, center=<1 0 0>, rotation=<rot 90 0 0>}]
 dyn_chain_b_02 [{shapeType=Capsule, length=1, radius=0.8, center=<1 0 0>, rotation=<rot 90 0 0>}]
 dyn_chain_b_03 [{shapeType=Capsule, length=1, radius=0.8, center=<1 0 0>, rotation=<rot 90 0 0>}]
 dyn_chain_b_04 [{shapeType=Capsule, length=1, radius=0.8, center=<1 0 0>, rotation=<rot 90 0 0>}]
 dyn_chain_b_05 [{shapeType=Capsule, length=1, radius=0.8, center=<1 0 0>, rotation=<rot 90 0 0>}]
 pack [{shapeType=Capsule, length=21.8875, radius=2.9201, center=<0 -10.6015 -10.199>, rotation=<rot 90 0 0>}]
}
constraints [
 {bone1Name=dyn_chain_a_01, bone2Name=pack, swing1LimitDegrees=30, swing2LimitDegrees=30, twistLimitDegrees=5}
 {bone1Name=dyn_chain_a_02, bone2Name=dyn_chain_a_01, swing1LimitDegrees=30, swing2LimitDegrees=30, twistLimitDegrees=5}
 {bone1Name=dyn_chain_a_03, bone2Name=dyn_chain_a_02, swing1LimitDegrees=30, swing2LimitDegrees=30, twistLimitDegrees=5}
 {bone1Name=dyn_chain_a_04, bone2Name=dyn_chain_a_03, swing1LimitDegrees=30, swing2LimitDegrees=30, twistLimitDegrees=5}
 {bone1Name=dyn_chain_a_05, bone2Name=dyn_chain_a_04, swing1LimitDegrees=30, swing2LimitDegrees=30, twistLimitDegrees=5}
 {bone1Name=dyn_chain_a_06, bone2Name=dyn_chain_a_05, swing1LimitDegrees=30, swing2LimitDegrees=30, twistLimitDegrees=5}
 {bone1Name=dyn_chain_a_07, bone2Name=dyn_chain_a_06, swing1LimitDegrees=30, swing2LimitDegrees=30, twistLimitDegrees=5}
 {bone1Name=dyn_chain_b_01, bone2Name=pack, swing1LimitDegrees=30, swing2LimitDegrees=30, twistLimitDegrees=5}
 {bone1Name=dyn_chain_b_02, bone2Name=dyn_chain_b_01, swing1LimitDegrees=30, swing2LimitDegrees=30, twistLimitDegrees=5}
 {bone1Name=dyn_chain_b_03, bone2Name=dyn_chain_b_02, swing1LimitDegrees=30, swing2LimitDegrees=30, twistLimitDegrees=5}
 {bone1Name=dyn_chain_b_04, bone2Name=dyn_chain_b_03, swing1LimitDegrees=30, swing2LimitDegrees=30, twistLimitDegrees=5}
 {bone1Name=dyn_chain_b_05, bone2Name=dyn_chain_b_04, swing1LimitDegrees=30, swing2LimitDegrees=30, twistLimitDegrees=5}
 {
  bone1Name=dyn_chain_b_05
  bone2Name=dyn_chain_a_07
  swing1LimitDegrees=15
  swing1Motion=Locked
  swing2LimitDegrees=15
  swing2Motion=Locked
  twistLimitDegrees=5
  twistMotion=Locked
 }
]
