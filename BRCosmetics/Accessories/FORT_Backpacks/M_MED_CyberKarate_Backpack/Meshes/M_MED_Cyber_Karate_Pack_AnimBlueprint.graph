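# Animation Blueprint for the Cyber Karate backpack cosmetic, adding secondary motion to its dangling bone chains.
# The input pose gets a slight additive offset on the pack bone and is cached for two branches.
# bRigidBodyEnabled, not declared here, selects RigidBody simulation with overridden gravity and physics asset, or chained Trail controllers on dyn_chain_a and dyn_chain_b.
# The event graph has only unconnected stub nodes.
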AnimBlueprint /BRCosmetics/Accessories/FORT_Backpacks/M_MED_CyberKarate_Backpack/Meshes/M_MED_Cyber_Karate_Pack_AnimBlueprint
target_skeleton=&/M_MED_Cyber_Karate_Pack_Skeleton
graphs [
 AnimGraph {nodes [
   n1 = RigidBody(ComponentPose=Local To Component(LocalPose=Use cached pose 'In Pose'())) {
    bClampLinearTranslationLimitToRefPose=true
    bOverrideWorldGravity=true
    overridePhysicsAsset=&/M_MED_Cyber_Karate_Pack_Physics
    alphaBoolBlend {blendOption=Linear}
    componentAppliedLinearAccClamp=<1000 1000 1000>
    componentLinearAccScale=<0.3 0.3 0.3>
    overrideWorldGravity {z=-980}
   }
   n2 = Trail controller - Bone: dyn_chain_a_07(ComponentPose=Local To Component(LocalPose=Use cached pose 'In Pose'()), Alpha=1) {
    chainLength=4
    relaxationSpeedScale=3
    rotationLimits [{limitMax=<180 180 180>}, {limitMax=<180 180 180>}, {limitMax=<180 180 180>}, {limitMax=<180 180 180>}]
    rotationOffsets [<0 0 0>, <0 0 0>, <0 0 0>, <0 0 0>]
    trailBone {boneName=dyn_chain_a_07}
   }
   n3 = Trail controller - Bone: dyn_chain_b_05(ComponentPose=@n2, Alpha=1) {
    chainLength=4
    relaxationSpeedScale=3
    rotationLimits [{limitMax=<180 180 180>}, {limitMax=<180 180 180>}, {limitMax=<180 180 180>}, {limitMax=<180 180 180>}]
    rotationOffsets [<0 0 0>, <0 0 0>, <0 0 0>, <0 0 0>]
    trailBone {boneName=dyn_chain_b_05}
   }
   Output Pose(Result=Blend Poses by bool(BlendPose_0=Component To Local(ComponentPose=@n1), BlendPose_1=Component To Local(ComponentPose=@n3), bActiveValue=Get bRigidBodyEnabled()) {blendType=Linear}) {layerGroup=None}
   AnimGraphNode_SubInput_0 = Input Pose()
   n4 = "Transform (Modify) Bone - Bone: pack"(ComponentPose=Local To Component(LocalPose=@AnimGraphNode_SubInput_0), Translation="0.000,-0.5,0.000", Rotation="0,0,0") {
    translationMode=BMM_Additive
    translationSpace=BCS_BoneSpace
    alphaBoolBlend {blendOption=Linear}
    boneToModify {boneName=pack}
    translation {y=-0.5}
   }
   Save cached pose 'In Pose'(Pose=Component To Local(ComponentPose=@n4)) {cacheName=In Pose}
  ]}
 EventGraph {nodes [
   TryGetPawnOwner()
   Event BlueprintUpdateAnimation()
  ]}
]
variables []
